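# Translucent unlit overlay material for the BananaFlare cosmetic shoe, reusing NanaEternal noise and translucent blur functions between the character parent PreFX and PostFX stages.
# A selectable TechArt mask channel scales opacity, and an additional mask adds squared stitch color to emissive.
# Emissive dims where the normal deviates from flat.
# World position offset pulls masked vertices to the object position, collapsing that geometry.
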Material /CosmeticShoes/Assets/FORT_Shoes/Shoe_BananaFlare/Materials/M_NanaEternal_Overlay
graph {
 properties {
  bUseMaterialAttributes=true
  bUsedWithClothing=true
  bUsedWithMeshDeformer=true
  bUsedWithMorphTargets=true
  bUsedWithNanite=true
  bUsedWithNiagaraMeshParticles=true
  bUsedWithSkeletalMesh=true
  blendMode=BLEND_Translucent
  pixelDepthOffsetMode=PDOM_Legacy
  shadingModel=MSM_Unlit
  usedShadingModels=Unlit
 }
 nodes [
  n1 = MaterialFunctionCall() {materialFunction=&MF_CharacterParent_2023_PreFX}
  n2 = MaterialFunctionCall!(MaterialAttributes=@n1.Result, Fresnel=@n1.Result) {materialFunction=/BRCosmetics/Characters/Player/Male/Medium/Bodies/M_MED_NanaEternal/Materials/MF_NanaEternal_Noise}
  n3 = GetMaterialAttributes(None=MaterialFunctionCall!(Material=@n2.Result, Noise=@n2.Result) {materialFunction=/BRCosmetics/Characters/Player/Male/Medium/Bodies/M_MED_NanaEternal/Materials/MF_NanaEternal_TransBlur}.Result) {attributeGetTypes [#1, #2, #3]}
  n4 = TextureSampleParameter2D() {
   group=TwosidedCull
   parameterName=TechArtMask_RE
   texture=/CosmeticShoes/Assets/FORT_Shoes/Shoe_BananaFlare/Textures/T_BananaFlare_Shoe_TechArt
  }
  n5 = ChannelMaskParameter(None=@n4.RGB) {group=TwosidedCull, parameterName=CullOpacityMask}
  n6 = TextureSampleParameter2D() {
   group=TwosidedCull
   parameterName=TechArtMask_Additional
   samplerType=SAMPLERTYPE_Masks
   texture=/CosmeticShoes/Assets/FORT_Shoes/Shoe_BananaFlare/Textures/T_BananaFlare_Shoe_AdditionalMask
  }
  n7 = VectorParameter() {group=TwosidedCull, parameterName=StitchesColor, defaultValue=<rgba 0.3355 0.8031 0.3755 1>}
  n8 = Add(A=@n3.Emissive Color, B=Multiply(A=@n6.G, B=Multiply(A=@n7.RGB, B=@n7.RGB)))
  n9 = Distance(A=@n3.Normal, B=Constant3Vector() {constant=<rgba 0 0 1 0>})
  n10 = LinearInterpolate(Alpha=Saturate(None=Multiply(A=@n9) {constB=3})) {constA=1, constB=0.5}
  n11 = Subtract(A=ObjectPositionWS(), B=WorldPosition().XYZ)
  n12 = Saturate(None=Add(A=@n6.B, B=@n4.G))
  MaterialFunctionCall(
   Material=SetMaterialAttributes!(
    MaterialAttributes=@n3.MaterialAttributes
    Opacity=Multiply(A=@n3.Opacity, B=@n5.R)
    Emissive Color=Multiply(A=@n8, B=@n10)
    World Position Offset=Multiply(A=@n11, B=@n12)
   ) {attributeSetTypes [#1, #2, #4]}
   M=@n1.M
   Pre Skinned Local Position=@n1.M
   Distance Blend=@n1.M
   Fresnel=@n1.M
  ) {materialFunction=&MF_CharacterParent_2023_PostFX}
 ]
}
