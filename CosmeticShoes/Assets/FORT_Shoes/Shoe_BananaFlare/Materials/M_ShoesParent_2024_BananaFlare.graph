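# Masked, two-sided parent material for the BananaFlare cosmetic shoe, built on the 2023 character PreFX/PostFX functions plus shoe reactivity, sock and hologram functions.
# Tech art masks and TwoSidedSign cut opacity by face side, while ShadowReplace keeps shadows solid.
# Base color adds an emissive texture and has an area darkened.
# Named reroutes carry masked flattened-normal Fresnel color and a sine-shaped Fresnel emissive term.
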Material /CosmeticShoes/Assets/FORT_Shoes/Shoe_BananaFlare/Materials/M_ShoesParent_2024_BananaFlare
graph {
 properties {
  bUseMaterialAttributes=true
  bUsedWithClothing=true
  bUsedWithMeshDeformer=true
  bUsedWithMorphTargets=true
  bUsedWithNanite=true
  bUsedWithNiagaraMeshParticles=true
  bUsedWithSkeletalMesh=true
  blendMode=BLEND_Masked
  pixelDepthOffsetMode=PDOM_Legacy
  shadingModel=MSM_FromMaterialExpression
  twoSided=true
  usedShadingModels=Default Lit | Subsurface | Clear Coat | Subsurface Profile | Cloth
 }
 nodes [
  n1 = MaterialFunctionCall() {materialFunction=&MF_CharacterParent_2023_PreFX}
  n2 = GetMaterialAttributes(None=MaterialFunctionCall!(
    InOpacity
    Material=MaterialFunctionCall!(Material=@n1.Result) {materialFunction=/Game/Global/Materials/Cosmetics/Kicks/Shoes/Materials/MF_ShoeReactivityTest}.Result
    Mask=@n1.Tech Art Mask
   ) {materialFunction=/Game/Global/Materials/Cosmetics/Kicks/Socks/Materials/MF_Sock}.Result) {attributeGetTypes [#1, #2]}
  n3 = Saturate(None=Sign(None=TwoSidedSign()))
  n4 = TextureSampleParameter2D() {
   group=TwosidedCull
   parameterName=TechArtMask_RE
   texture=/CosmeticShoes/Assets/FORT_Shoes/Shoe_BananaFlare/Textures/T_BananaFlare_Shoe_TechArt
  }
  n5 = LinearInterpolate(B=OneMinus(None=@n3), Alpha=ChannelMaskParameter(None=@n4.RGB) {group=TwosidedCull, parameterName=CullOpacityMask}.R) {constA=1}
  n6 = NamedRerouteUsage()
  n7 = ShadowReplace(Default=LinearInterpolate(A=LinearInterpolate(A=@n5, B=@n3, Alpha=@n6), Alpha=@n4.B) {constB=0}, Shadow=Constant() {r=1})
  n8 = TextureSampleParameter2D() {
   group=TwosidedCull
   parameterName=EmissiveToBaseColor
   texture=/CosmeticShoes/Assets/FORT_Shoes/Shoe_BananaFlare/Textures/T_BananaFlare_Shoe_E
  }
  n9 = Multiply(A=@n8.RGB, B=ScalarParameter() {defaultValue=4, parameterName=BaseColor_EM_Multi})
  n10 = Saturate(None=Add(A=Add(A=@n2.Base Color, B=@n9), B=NamedRerouteUsage()))
  n11 = LinearInterpolate(A=@n10, B=Multiply(A=@n10, B=ScalarParameter() {defaultValue=0.5, group=TwosidedCull, parameterName=DarkenArea}), Alpha=NamedRerouteUsage())
  n12 = NamedRerouteUsage()
  MaterialFunctionCall!(In=MaterialFunctionCall(
    Material=SetMaterialAttributes!(MaterialAttributes=@n2.MaterialAttributes, Opacity Mask=@n7, Base Color=@n11, Emissive Color=@n12) {attributeSetTypes [#3, #1, #4]}
    M=@n1.M
    Pre Skinned Local Position=@n1.M
    Distance Blend=@n1.M
    Fresnel=@n1.M
   ) {materialFunction=&MF_CharacterParent_2023_PostFX}.Result) {materialFunction=/Game/Global/Materials/Cosmetics/Kicks/Shoes/Materials/MF_TechartShoeHologram}
  NamedRerouteDeclaration(None=@n1.Pre skinned Local Position) {name=PreskinnedPos, nodeColor=<rgba 1 0 0.2578 1>}
  NamedRerouteDeclaration(None=@n2.Normal) {name=Normal_IN, nodeColor=<rgba 0 0.7422 1 1>}
  n13 = VectorParameter() {parameterName=FresnelColor, defaultValue=<rgba 0 1 0.2427 2>}
  n14 = Multiply(A=@n13.RGB, B=@n13.RGB)
  n15 = NamedRerouteUsage()
  n16 = ScalarParameter() {defaultValue=-1, group=TwosidedCull, parameterName=FlattenN}
  n17 = Fresnel(
   ExponentIn=ScalarParameter() {defaultValue=2, group=TwosidedCull, parameterName=FrenselPower}
   Normal=Transform(None=MaterialFunctionCall(Normal=@n15, Flatness=@n16) {materialFunction=&FlattenNormal}.Result)
  ) {baseReflectFraction=0}
  NamedRerouteDeclaration(None=Multiply(
    A=Multiply(A=@n14, B=@n17)
    B=ChannelMaskParameter(None=@n4.RGB) {group=TwosidedCull, maskChannel=Green, parameterName=SkeletonMask, defaultValue=<rgba 0 1 0 0>}.R
   )) {name=FresnelBaseColor, nodeColor=<rgba 1 0.9375 0 1>}
  NamedRerouteDeclaration(None=ChannelMaskParameter(None=@n1.Tech Art Mask) {group=TwosidedCull, parameterName=AreaMask}.R) {name=AreaMask, nodeColor=<rgba 0.5703 0 1 1>}
  n18 = TextureSampleParameter2D() {
   group=TwosidedCull
   parameterName=TechArtMask_Additional
   samplerType=SAMPLERTYPE_Masks
   texture=/CosmeticShoes/Assets/FORT_Shoes/Shoe_BananaFlare/Textures/T_BananaFlare_Shoe_AdditionalMask
  }
  NamedRerouteDeclaration(None=@n18.R) {name=AreaMask_Comp, nodeColor=<rgba 0 1 0.25 1>}
  NamedRerouteDeclaration(None=@n14) {name=FresnelColor, nodeColor=<rgba 0 1 0.1797 1>}
  n19 = Multiply(A=NamedRerouteUsage(), B=ScalarParameter() {defaultValue=0.2, group=TwosidedCull, parameterName="EmissiveOr=verride"})
  n20 = Add(
   A=Multiply(A=Fresnel() {baseReflectFraction=0, exponent=1}) {constB=2}
   B=ScalarParameter() {defaultValue=0.5, group=TwosidedCull, parameterName=Offset}
  )
  n21 = Power(Base=ConstantBiasScale(None=Sine(None=@n20)))
  NamedRerouteDeclaration(None=Multiply(A=Multiply(A=@n19, B=@n21), B=NamedRerouteUsage())) {name=AdditionalEmissive, nodeColor=<rgba 1 0 0.1406 1>}
  NamedRerouteDeclaration(None=@n18.B) {name=Area_Positive, nodeColor=<rgba 1 0.0938 0 1>}
 ]
}
